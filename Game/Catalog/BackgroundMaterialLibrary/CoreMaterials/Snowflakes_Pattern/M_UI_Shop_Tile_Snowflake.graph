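# Translucent UI material for Item Shop tile backgrounds, layering an animated pattern, sparkles and sheen over color gradients, with rounded-corner masking.
# Tile background params supply three colors for the gradients and accents.
# A particle-grid function repeats an alpha texture (clover by default, despite the Snowflake name) with scrolling and per-cell random scale, rotation and opacity.
# Sparkles combine a flake noise texture, panning blurred noise and procedural twinkling stars, alongside a delayed sweeping sheen.
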Material /Game/Catalog/BackgroundMaterialLibrary/CoreMaterials/Snowflakes_Pattern/M_UI_Shop_Tile_Snowflake
graph {
 properties {bAutomaticallySetUsageInEditor=false, blendMode=BLEND_Translucent, materialDomain=MD_UI, usedShadingModels=Default Lit}
 nodes [
  n1 = MaterialFunctionCall!(ScreenUV, TextureSize, ScalePivotPoint) {materialFunction=/FortUILibrary/MaterialFunctions/Basics/MF_UI_ScreenSpaceUVScaling}
  n2 = NamedRerouteUsage()
  n3 = NamedRerouteUsage()
  n4 = NamedRerouteUsage()
  n5 = Frac(None=Multiply(A=ScalarParameter() {defaultValue=0.015, group=02 - Pattern, parameterName=Pattern Offset Speed}, B=@n4) {constB=0.015})
  n6 = NamedRerouteUsage()
  n7 = ScalarParameter() {defaultValue=0.015, group=03 - Cell, parameterName=Cell Rotation Speed}
  n8 = Add(A=ScalarParameter() {group=03 - Cell, parameterName=Cell Default Rotation, sliderMax=1}, B=Frac(None=Multiply(A=@n4, B=@n7)))
  n9 = MaterialFunctionCall!(
   UV=Multiply(
    A=StaticSwitchParameter(True=@n1.Scaled to Fit, False=@n2) {group=02 - Pattern, parameterName=Screen Space Pattern}
    B=ScalarParameter() {defaultValue=0.5, group=02 - Pattern, parameterName=Pattern Scale}
   )
   Global Offset=AppendVector(A=Add(A=@n3, B=@n5), B=Constant())
   Global Rotation=ScalarParameter() {defaultValue=0.35, group=02 - Pattern, parameterName=Pattern Rotation}
   Tiling=Floor(None=ScalarParameter() {defaultValue=5, group=02 - Pattern, parameterName=Pattern Tiling})
   Cell Offset=Constant()
   Cell Scale=LinearInterpolate(
    A=ScalarParameter() {defaultValue=0.5, group=03 - Cell, parameterName=Minimum Cell Scale}
    B=ScalarParameter() {defaultValue=1, group=03 - Cell, parameterName=Maximum Cell Scale}
    Alpha=@n6
   ) {constA=0.5}
   Cell Rotation=LinearInterpolate(A=@n8, B=Add(A=@n6, B=@n8), Alpha=ScalarParameter() {defaultValue=1, group=0 - Global, parameterName=Randomize Rotation, sliderMax=1})
   Random Seed
  ) {materialFunction=/FortUILibrary/MaterialFunctions/Basics/MF_UI_ParticleGrid}
  n10 = TextureObjectParameter() {
   group=02 - Pattern
   parameterName=Pattern Texture
   samplerType=SAMPLERTYPE_Alpha
   texture=/Game/UI/Frontend/ItemShop/Art/Lucky/T_UI_LuckyClover-01
  }
  n11 = TextureSample(UVs=NamedRerouteDeclaration(None=@n9.Cells UV) {name=Cells UV, nodeColor=<rgba 0 0 0 1>}, Tex=@n10)
  n12 = NamedRerouteUsage()
  n13 = Constant() {r=1.5}
  n14 = MaterialFunctionCall(UV=MaterialFunctionCall(UV=MaterialFunctionCall(UVs=@n12, Texture Scale=@n13) {materialFunction=&ScaleUVsByCenter}.UVs) {materialFunction=&MF_UI_UVScaling}.Scaled To Fill) {materialFunction=&MF_UI_RadialGradient}
  n15 = NamedRerouteUsage()
  n16 = Max(A=Round(None=@n15), B=@n15)
  n17 = NamedRerouteUsage()
  n18 = Multiply(A=NamedRerouteUsage(), B=ScalarParameter() {defaultValue=1, group=03 - Cell, parameterName=Cell Opacity Fluctuation Speed}) {constB=0.15}
  n19 = Saturate(None=MaterialFunctionCall(Speed=LinearInterpolate(B=@n17, Alpha=@n18) {constAlpha=1}) {materialFunction=&DebugTimeSine}.Result)
  n20 = ScalarParameter() {group=03 - Cell, parameterName=Fluctuate Cell Opacity, sliderMax=1}
  n21 = Multiply(
   A=Min(A=NamedRerouteDeclaration(None=@n11.R) {name=Pattern Mask, nodeColor=<rgba 0 0 0 1>}, B=Multiply(A=@n14.Smooth))
   B=LinearInterpolate(
    A=ScalarParameter() {defaultValue=0.15, group=03 - Cell, parameterName=Cell Min Opacity, sliderMax=1}
    B=ScalarParameter() {defaultValue=1, group=03 - Cell, parameterName=Cell Max Opacity, sliderMax=1}
    Alpha=LinearInterpolate(A=@n16, B=@n19, Alpha=@n20)
   ) {constA=0.15}
  )
  n22 = NamedRerouteUsage()
  Add(A=@n21, B=Divide(A=Min(A=@n21, B=@n22)) {constB=0.2})
  n23 = NamedRerouteUsage()
  n24 = VectorParameter() {group=Sparkles, parameterName=SparkleColor, sortPriority=1, defaultValue=<rgba 0.6308 0.2799 0.001 1>}
  n25 = Saturate(None=Multiply(A=NamedRerouteDeclaration(None=@n24.RGB) {name=Sparkles color, nodeColor=<rgba 0 0 0 1>}) {constB=0.7})
  n26 = NamedRerouteUsage()
  n27 = NamedRerouteUsage()
  n28 = Add(A=NamedRerouteUsage(), B=NamedRerouteUsage()) {constB=0.7}
  n29 = NamedRerouteUsage()
  n30 = ScalarParameter() {defaultValue=32, group=Sparkles, parameterName=SparklesIntensity, sortPriority=3}
  n31 = Multiply(A=Saturate(None=Multiply(A=@n29, B=@n30)), B=NamedRerouteUsage())
  LinearInterpolate(A=NamedRerouteUsage(), B=LinearInterpolate(A=LinearInterpolate(A=@n23, B=@n25, Alpha=@n26), B=Saturate(None=@n28), Alpha=@n27), Alpha=@n31)
  n32 = NamedRerouteUsage()
  n33 = NamedRerouteUsage()
  LinearInterpolate(
   A=@n32
   B=Saturate(None=MaterialFunctionCall!(Base=@n32, Add=@n33) {materialFunction=/Engine/Functions/Engine_MaterialFunctions02/Math/MultiplyAdd}.Result)
   Alpha=LinearInterpolate(B=NamedRerouteUsage(), Alpha=ScalarParameter() {defaultValue=1, group=04 - Sheen, parameterName=Apply Sheen on Pattern, sliderMax=1}) {constA=1}
  )
  MaterialFunctionCall(UVs=NamedRerouteUsage(), RotationAngle=ScalarParameter() {group=01 - Gradient, parameterName=Gradient Rotation}) {materialFunction=&MF_UI_LinearGradient}
  MaterialFunctionCall!(Input Alpha=Constant() {r=1}, bOverrideCornerRadius, Corner Radius Override) {materialFunction=/FortUILibrary/Systems/ItemShopTile/Materials/MaterialFunctions/MF_UI_ItemShop_RoundedCorners}
  n34 = Multiply(
   A=MaterialFunctionCall() {materialFunction=&MF_UI_UVScaling}.Scaled to Fit
   B=ScalarParameter() {defaultValue=16, parameterName=Tilling amount}
  ) {constB=20}
  n35 = Frac(None=@n34)
  n36 = ComponentMask(None=@n35) {r=true}
  n37 = ComponentMask(None=@n35) {g=true}
  n38 = Multiply(A=Abs(None=Subtract(A=@n36) {constB=0.5}), B=Abs(None=Subtract(A=@n37) {constB=0.5}))
  Min(
   A=OneMinus(None=SmoothStep(Value=@n38) {constMax=0.005})
   B=MaterialFunctionCall(UV=@n35, GradientStart=Constant() {r=0.3}, GradientEnd=Constant() {r=1}) {materialFunction=&MF_UI_RadialGradient}.Smooth
  )
  n39 = MaterialFunctionCall() {materialFunction=/FortUILibrary/Systems/ItemShopTile/Materials/MaterialFunctions/MF_UI_ItemShop_BackgroundParams}
  NamedRerouteDeclaration(None=@n39.Color 3) {name=Accent Color, nodeColor=<rgba 0 0 0 1>}
  NamedRerouteDeclaration(None=@n39.Color 1) {name=Primary Color, nodeColor=<rgba 0 0 0 1>}
  n40 = NamedRerouteUsage()
  n41 = ScalarParameter() {defaultValue=0.8, group=Sparkles, parameterName=BGSparkle Opacity}
  n42 = NamedRerouteUsage()
  n43 = Add(
   A=MaterialFunctionCall(UVs=NamedRerouteUsage(), Texture Scale=ScalarParameter() {defaultValue=1, group=01 - Gradient, parameterName=Radial Gradient Scale}) {materialFunction=&ScaleUVsByCenter}.UVs
   B=AppendVector(
    A=ScalarParameter() {group=01 - Gradient, parameterName=Radial Gradient Offset X}
    B=ScalarParameter() {group=01 - Gradient, parameterName=Radial Gradient Offset Y}
   )
  ) {constB=0.5}
  n44 = Saturate(None=MaterialFunctionCall(UV=MaterialFunctionCall(UV=@n43) {materialFunction=&MF_UI_UVScaling}.Scaled To Fill) {materialFunction=&MF_UI_RadialGradient}.Smooth)
  n45 = LinearInterpolate(A=NamedRerouteUsage(), B=@n42, Alpha=@n44)
  n46 = Multiply(A=NamedRerouteUsage()) {constB=1.5}
  NamedRerouteDeclaration(None=Add(A=Multiply(A=@n40, B=@n41), B=LinearInterpolate(A=@n45, B=@n42, Alpha=@n46))) {name=Colored Gradient, nodeColor=<rgba 0 0 0 1>}
  NamedRerouteDeclaration(None=@n21) {name=Pattern Opacity, nodeColor=<rgba 0 0 0 1>}
  n47 = NamedRerouteUsage()
  n48 = ScalarParameter() {group=04 - Sheen, parameterName=Sheen Angle}
  n49 = LinearInterpolate(Alpha=MaterialFunctionCall!(
    Animation Duration=ScalarParameter() {defaultValue=1, group=04 - Sheen, parameterName=Sheen Duration}
    Delay=ScalarParameter() {defaultValue=3, group=04 - Sheen, parameterName=Sheen Delay}
    Reset to Default
   ) {materialFunction=/FortUILibrary/MaterialFunctions/Time/MF_UI_AnimationDelay}.Result) {constA=-1}
  n50 = Constant()
  n51 = MaterialFunctionCall(
   UVs=Add(A=Rotator(Coordinate=@n47, Time=@n48) {speed=6.2832}, B=AppendVector(A=@n49, B=@n50))
   Texture Scale=AppendVector(A=ScalarParameter() {defaultValue=0.2, group=04 - Sheen, parameterName=Sheen Scale}, B=Constant() {r=1})
  ) {materialFunction=&ScaleUVsByCenter}
  n52 = ComponentMask(None=MaterialFunctionCall!(Value=@n51.UVs, Scale=Constant() {r=1}, Height=Constant() {r=1}) {materialFunction=/FortUILibrary/MaterialFunctions/Basics/MF_UI_PingPong}.Result) {r=true}
  NamedRerouteDeclaration(None=Multiply(
    A=MaterialFunctionCall!(UVs=@n51.UVs, Alpha=@n52) {materialFunction=/FortUILibrary/MaterialFunctions/Basics/MF_UI_ClipEdgePixel}.Result
    B=ScalarParameter() {defaultValue=1, group=04 - Sheen, parameterName=Sheen Opacity, sliderMax=1}
   )) {name=Sheen Opacity, nodeColor=<rgba 0 0 0 1>}
  n53 = NamedRerouteUsage()
  n54 = Multiply(A=NamedRerouteUsage(), B=ScalarParameter() {defaultValue=1, group=Sparkles, parameterName=Starshine_Intensity})
  n55 = NamedRerouteUsage()
  n56 = ScalarParameter() {defaultValue=1, group=02 - Pattern, parameterName=Pattern Blend, sliderMax=1, sliderMin=-1}
  n57 = NamedRerouteUsage()
  n58 = Saturate(None=MaterialFunctionCall(Base=@n53, Add=Multiply(A=@n56, B=@n57) {constB=2}) {materialFunction=&MultiplyAdd}.Result)
  n59 = ScalarParameter() {group=02 - Pattern, parameterName=Blend Pattern and Gradient, sliderMax=1}
  NamedRerouteDeclaration(None=LinearInterpolate(A=Add(A=@n53, B=@n54), B=LinearInterpolate(A=@n55, B=@n58, Alpha=@n59), Alpha=NamedRerouteUsage())) {name=Pattern Color, nodeColor=<rgba 0 0 0 1>}
  n60 = ScalarParameter() {defaultValue=1, group=0 - Global, parameterName=Global Speed}
  NamedRerouteDeclaration(None=@n60) {name=Global Speed Input, nodeColor=<rgba 0 0 0 1>}
  n61 = MaterialFunctionCall(UV=NamedRerouteDeclaration(None=MaterialFunctionCall() {materialFunction=&GetUserInterfaceUV}.Normalized UV) {name=Normalized UV, nodeColor=<rgba 0 0 0 1>}) {materialFunction=&MF_UI_UVScaling}
  NamedRerouteDeclaration(None=@n61.Scaled to Fit) {name=Scale to Fill UV, nodeColor=<rgba 0 0 0 1>}
  n62 = NamedRerouteUsage()
  n63 = AppendVector(A=ScalarParameter() {defaultValue=0.15, group=Sparkles, parameterName=SparklesSpeedSheen, sortPriority=5}, B=Constant())
  n64 = Multiply(
   A=TextureSample(UVs=Panner(Coordinate=@n62, Speed=@n63)) {samplerType=SAMPLERTYPE_Grayscale, texture=&LowResBlurredNoise}.R
   B=NamedRerouteUsage()
  )
  NamedRerouteDeclaration(None=Saturate(None=Multiply(A=@n64) {constB=2})) {name=NoiseCloudsSheenMask, nodeColor=<rgba 0 0 0 1>}
  NamedRerouteDeclaration(None=Min(A=@n31, B=@n27)) {name=SparklesSystem, nodeColor=<rgba 0 0 0 1>}
  NamedRerouteDeclaration(None=@n39.Color 2) {name=Secondary Color, nodeColor=<rgba 0 0 0 1>}
  n65 = NamedRerouteUsage()
  n66 = Constant() {r=-0.1}
  n67 = Constant() {r=0.27}
  NamedRerouteDeclaration(None=Saturate(None=MaterialFunctionCall(UVs=@n65, GradientStart=@n66, GradientEnd=@n67) {materialFunction=&MF_UI_LinearGradient}.Result)) {name=Sparks Gradient Mask, nodeColor=<rgba 0 0 0 1>}
  n68 = Floor(None=@n34)
  NamedRerouteDeclaration(None=MaterialFunctionCall!(UV=ConstantBiasScale(None=@n68), Seed=LinearInterpolate() {constA=0.5}) {materialFunction=&MF_UI_Random}.Result) {name=randomd cell vaules, nodeColor=<rgba 0.0781 0 1 1>}
  n69 = MaterialFunctionCall() {materialFunction=&MF_UI_ScreenSpaceUVScaling}
  n70 = ScalarParameter() {defaultValue=1, group=Sparkles, parameterName=SheenScale1}
  n71 = Panner(
   Coordinate=MaterialFunctionCall(
    UVs=MaterialFunctionCall(UVs=@n69.Scaled to Fill, Texture Scale=@n70) {materialFunction=&ScaleUVsByCenter}.UVs
    "Rotation Angle (0-1)"=ScalarParameter() {defaultValue=0.1, group=Sparkles, parameterName=SheenRotation1}
   ) {materialFunction=&CustomRotator}.Rotated Values
   Time=Time()
   Speed=ScalarParameter() {defaultValue=0.1, group=Sparkles, parameterName=SheenSpeed1}
  ) {speedX=-0.08}
  n72 = MaterialFunctionCall("Value (0-1)"=ComponentMask(None=Sine(None=@n71) {period=2}) {r=true}) {materialFunction=&MF_UI_EaseCurves}
  NamedRerouteDeclaration(None=Saturate(None=MaterialFunctionCall("Value (0-1)"=@n72.EaseInQuint) {materialFunction=&MF_UI_EaseCurves}.EaseInOutSine)) {name=SheenLight, nodeColor=<rgba 0 0 0 1>}
  n73 = ScalarParameter() {defaultValue=0.1, group=Sparkles, parameterName=SparklesSpeedColor, sortPriority=6}
  n74 = Constant() {r=0.03}
  n75 = TextureSample(UVs=Panner(Coordinate=NamedRerouteUsage(), Speed=AppendVector(A=@n73, B=@n74)) {speedX=0.1}) {samplerType=SAMPLERTYPE_Grayscale, texture=&LowResBlurredNoise}
  NamedRerouteDeclaration(None=Saturate(None=Multiply(A=@n75.R) {constB=3})) {name=NoiseCloudMaskSimple, nodeColor=<rgba 0 0 0 1>}
  n76 = MaterialFunctionCall(UV=MaterialFunctionCall(UVs=NamedRerouteUsage(), Texture Scale=Constant() {r=1.2}) {materialFunction=&ScaleUVsByCenter}.UVs) {materialFunction=&MF_UI_UVScaling}
  n77 = NamedRerouteUsage()
  n78 = Multiply(A=Abs(None=Subtract(A=@n77) {constB=0.5})) {constB=2}
  n79 = MaterialFunctionCall(UVs=@n78) {materialFunction=&ScaleUVsByCenter}
  n80 = MaterialFunctionCall(UVs=@n78) {materialFunction=&ScaleUVsByCenter}
  n81 = OneMinus(None=LinearInterpolate(A=ComponentMask(None=@n79.UVs) {g=true}, Alpha=ComponentMask(None=@n80.UVs) {r=true}))
  n82 = Constant() {r=1}
  n83 = Constant() {r=600}
  n84 = Constant() {r=4}
  n85 = NamedRerouteUsage()
  n86 = Multiply(A=Power(Base=Multiply(A=@n81, B=@n82), Exp=LinearInterpolate(A=@n83, B=@n84, Alpha=@n85)))
  NamedRerouteDeclaration(None=Min(
    A=MaterialFunctionCall(UV=@n76.Scaled To Fill) {materialFunction=&MF_UI_RadialGradient}.Smooth
    B=Saturate(None=@n86)
   )) {name=star, nodeColor=<rgba 0 1 0.1328 1>}
  n87 = Multiply(A=Time(), B=ScalarParameter() {defaultValue=0.2, parameterName=Speed})
  n88 = NamedRerouteUsage()
  n89 = LinearInterpolate(
   A=ScalarParameter() {parameterName=MinSize}
   B=ScalarParameter() {defaultValue=0.98, parameterName=MaxSize}
   Alpha=MaterialFunctionCall("Value (0-1)"=MaterialFunctionCall(Value=@n87, Scale=@n88) {materialFunction=&MF_UI_PingPong}.Result) {materialFunction=&MF_UI_EaseCurves}.EaseInOutSine
  )
  NamedRerouteDeclaration(None=@n89) {name=Aim_Sparkle, nodeColor=<rgba 1 0 0.75 1>}
  NamedRerouteDeclaration(None=@n35) {name=Tilling UVs, nodeColor=<rgba 0 0.2734 1 1>}
  NamedRerouteDeclaration(None=@n89) {name=Aim_slow sparkle, nodeColor=<rgba 1 0 0.3281 1>}
  NamedRerouteDeclaration(None=@n61.Scaled to Fit) {name=Scale Y UV, nodeColor=<rgba 0 0 0 1>}
  n90 = TextureCoordinate()
  n91 = Subtract(A=ComponentMask(None=@n90) {r=true}) {constB=0.5}
  n92 = Subtract(A=ComponentMask(None=@n90) {g=true}) {constB=0.5}
  n93 = SmoothStep(Value=Multiply(A=Abs(None=@n91), B=Abs(None=@n92))) {constMax=0.005}
  n94 = Constant() {r=0.3}
  n95 = Constant() {r=1}
  NamedRerouteDeclaration(None=Min(
    A=OneMinus(None=@n93)
    B=MaterialFunctionCall(UV=@n90, GradientStart=@n94, GradientEnd=@n95) {materialFunction=&MF_UI_RadialGradient}.Smooth
   )) {name=Stars, nodeColor=<rgba 0 1 0.625 1>}
  NamedRerouteDeclaration(None=Multiply(A=NamedRerouteUsage(), B=NamedRerouteUsage())) {name=StarShines, nodeColor=<rgba 1 0.3047 0 1>}
  NamedRerouteDeclaration(None=@n44) {name=BGRadialGradientMAsk, nodeColor=<rgba 0 0.6016 1 1>}
  NamedRerouteDeclaration(None=@n9.Random Rows) {name=Random Rows, nodeColor=<rgba 0 0 0 1>}
  NamedRerouteDeclaration(None=Multiply(A=Time(), B=@n60)) {name=Global Speed, nodeColor=<rgba 0 0 0 1>}
  NamedRerouteDeclaration(None=@n9.Random Cell Values) {name=Random Cell Values, nodeColor=<rgba 0 0 0 1>}
  n96 = TextureSample(UVs=MaterialFunctionCall(UVs=NamedRerouteUsage(), Texture Scale=ScalarParameter() {defaultValue=0.408, group=Sparkles, parameterName=SparkleScale, sortPriority=2}) {materialFunction=&ScaleUVsByCenter}.UVs) {texture=&/T_Noise_Sparkle_Flake}
  n97 = TextureSample(UVs=NamedRerouteUsage()) {samplerType=SAMPLERTYPE_Grayscale, texture=&LowResBlurredNoise}
  NamedRerouteDeclaration(None=Multiply(A=NamedRerouteDeclaration(None=@n96.R) {name=StarsMask, nodeColor=<rgba 0 0 0 1>}, B=Power(Base=@n97.R))) {name=MaskedStarsMask, nodeColor=<rgba 0 0 0 1>}
  NamedRerouteUsage()
  NamedRerouteUsage()
  NamedRerouteUsage()
  ScalarParameter() {defaultValue=0.7, group=Sparkles, parameterName=SparkleShineSaturation, sortPriority=4}
 ]
}
